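# Physics asset for the Badass Cowboy Cactus pickaxe, giving its gear, two flowers and their petals bodies for secondary motion.
# A capsule covers the two-handed handle, and the gear is constrained to it with only one swing axis left unlocked.
# Each flower carries five capsule petals constrained with locked twist and small swing limit values.
# Neither flower body has a constraint to the handle.
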PhysicsAsset /BRCosmetics/Weapons/FORT_Melee/Pickaxe_BadassCowboyCactus/Meshes/Pickaxe_Badass_Cowboy_Cactus_Physics
bodies {
 Melee_TwoHanded_Handle [{shapeType=Capsule, length=83.4943, radius=12.532, center=<1.6131 -0.0159 51.3656>, rotation=<rot -1.9973 -90 0.4436>}]
 dyn_flower_a [{shapeType=Sphere, radius=2.0042}]
 dyn_flower_b [{shapeType=Capsule, length=0.2768, radius=0.9168, rotation=<rot 0 -90 0>}]
 dyn_gear [{shapeType=Capsule, length=0.4389, radius=2.832, center=<0.9828 0 -0.0002>, rotation=<rot 0 -89.9999 90>}]
 dyn_petal_a01 [{shapeType=Capsule, length=1.4123, radius=1.099, center=<0.893 -0.1137 0.0502>, rotation=<rot -2.2513 -89.9992 -91.7053>}]
 dyn_petal_a02 [{shapeType=Capsule, length=1.4123, radius=1.0956, center=<0.9706 -0.1889 0.0072>, rotation=<rot -1.7011 -89.9999 -90.7363>}]
 dyn_petal_a03 [{shapeType=Capsule, length=1.4123, radius=1.098, center=<0.8945 -0.1547 0.0179>, rotation=<rot 0.902 -89.9992 -87.7602>}]
 dyn_petal_a04 [{shapeType=Capsule, length=1.6868, radius=1.2111, center=<1.2444 -0.1329 0.0285>, rotation=<rot -4.0541 -89.9996 -91.524>}]
 dyn_petal_a05 [{shapeType=Capsule, length=1.4123, radius=1.1043, center=<0.6407 -0.0452 -0.0111>, rotation=<rot -3.7892 -89.9992 -92.1346>}]
 dyn_petal_b01 [{shapeType=Capsule, length=1.0574, radius=0.6315, center=<0.3962 -0.1024 0.062>, rotation=<rot 1.3456 -89.9995 -89.3178>}]
 dyn_petal_b02 [{shapeType=Capsule, length=1.0574, radius=0.6435, center=<0.3489 -0.0785 0.0402>, rotation=<rot -2.9512 -89.9999 -86.3907>}]
 dyn_petal_b03 [{shapeType=Capsule, length=1.0574, radius=0.6226, center=<0.3103 -0.1276 0.0484>, rotation=<rot -3.239 -89.9999 -86.1963>}]
 dyn_petal_b04 [{shapeType=Capsule, length=1.0574, radius=0.6103, center=<0.5472 -0.2041 0.0632>, rotation=<rot 1.9137 -89.9994 -87.1904>}]
 dyn_petal_b05 [{shapeType=Capsule, length=1.0574, radius=0.6446, center=<0.4692 -0.0212 0.064>, rotation=<rot -2.6084 -90 -85.6605>}]
}
constraints [
 {bone1Name=dyn_petal_a01, bone2Name=dyn_flower_a, swing1LimitDegrees=3, swing2LimitDegrees=1, twistLimitDegrees=45, twistMotion=Locked}
 {bone1Name=dyn_petal_a02, bone2Name=dyn_flower_a, swing1LimitDegrees=3, swing2LimitDegrees=1, twistLimitDegrees=45, twistMotion=Locked}
 {bone1Name=dyn_petal_a03, bone2Name=dyn_flower_a, swing1LimitDegrees=3, swing2LimitDegrees=1, twistLimitDegrees=45, twistMotion=Locked}
 {bone1Name=dyn_petal_a04, bone2Name=dyn_flower_a, swing1LimitDegrees=3, swing2LimitDegrees=1, twistLimitDegrees=45, twistMotion=Locked}
 {bone1Name=dyn_petal_a05, bone2Name=dyn_flower_a, swing1LimitDegrees=3, swing2LimitDegrees=1, twistLimitDegrees=45, twistMotion=Locked}
 {bone1Name=dyn_petal_b01, bone2Name=dyn_flower_b, swing1LimitDegrees=2, swing2LimitDegrees=0.5, twistLimitDegrees=45, twistMotion=Locked}
 {bone1Name=dyn_petal_b02, bone2Name=dyn_flower_b, swing1LimitDegrees=2, swing2LimitDegrees=0.5, twistLimitDegrees=45, twistMotion=Locked}
 {bone1Name=dyn_petal_b03, bone2Name=dyn_flower_b, swing1LimitDegrees=2, swing2LimitDegrees=0.5, twistLimitDegrees=45, twistMotion=Locked}
 {bone1Name=dyn_petal_b04, bone2Name=dyn_flower_b, swing1LimitDegrees=2, swing2LimitDegrees=0.5, twistLimitDegrees=45, twistMotion=Locked}
 {bone1Name=dyn_petal_b05, bone2Name=dyn_flower_b, swing1LimitDegrees=2, swing2LimitDegrees=0.5, twistLimitDegrees=45, twistMotion=Locked}
 {
  bone1Name=dyn_gear
  bone2Name=Melee_TwoHanded_Handle
  swing1LimitDegrees=90
  swing2LimitDegrees=45
  swing2Motion=Locked
  twistLimitDegrees=45
  twistMotion=Locked
 }
]
